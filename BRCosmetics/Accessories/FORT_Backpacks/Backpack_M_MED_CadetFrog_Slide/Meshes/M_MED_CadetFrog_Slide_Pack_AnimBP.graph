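# Animation Blueprint for the CadetFrog Slide backpack, adding procedural and physics motion to the pack.
# A Control Rig fed by movement states leads into an additive needle-bone roll, then world-space Anim Dynamics on the tube bone.
# The event graph derives idle and moving flags and, when sprinting, randomizes needle alpha after a delay, smoothed by interpolation.
# It also computes gravity, rigid-body and simulation-space values, which the AnimGraph never visibly reads.
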AnimBlueprint /BRCosmetics/Accessories/FORT_Backpacks/Backpack_M_MED_CadetFrog_Slide/Meshes/M_MED_CadetFrog_Slide_Pack_AnimBP
target_skeleton=&/M_MED_CadetFrog_Slide_Pack_Skeleton
graphs [
 AnimGraph {nodes [
   Anim Dynamics - Bone: dyn_tube!(ComponentPose, Alpha=1, GravityOverride:Vector="0,0,-980", ExternalForce="0,0,0") -> (Pose) {
    angularDampingOverride=0.9
    angularSpringConstant=4000
    bAngularSpring=true
    bOverrideAngularDamping=true
    bUseGravityOverride=true
    gravityScale=0.35
    simulationSpace=World
    alphaBoolBlend {blendOption=Linear}
    boundBone {boneName=dyn_tube}
    componentAppliedLinearAccClamp=<9500 9500 9500>
    componentLinearAccScale=<1 1 1>
    componentLinearVelScale=<1 1 1>
    gravityOverride {z=-980}
    physicsBodyDefinitions [{
      boundBone {boneName=dyn_tube}
      boxExtents=<20 20 20>
      constraintSetup {angularLimitsMax=<3 3 3>, angularLimitsMin=<-3 -2 -5>, angularTarget=<1 0 0>}
      localJointOffset=<2 0 0>
     }]
    simSpaceSettings {maxAngularAcceleration=35, maxAngularVelocity=0.7, simSpaceAngularAlpha=1}
   }
   n1 = Control Rig!(
    Source=Input Pose()
    Alpha
    Is Sprinting:bool=Get Sprinting() -> (Sprinting:bool)
    Is Skydive:bool=Get bIsSkydiveFloating()
    Is IDLE:bool=Get isIDLE() -> (isIDLE:bool)
    Is Emote:bool=false
    Is Skydive Dive:bool=Get bIsSkydiveDiveMode()
    Is Skydive Parachute:bool=Get bIsParachuteOpened()
    Is Grindrail:bool=Get bIsOnGrindRail()
    Is Falling:bool=Get bIsFalling()
    Is Tactical Sprint:bool=Get bIsTacticalSprinting()
    Is Zipline:bool=Get OnZipline() -> (OnZipline:bool)
   ) -> (Pose) {
    alphaBoolBlend {blendOption=Linear}
    controlRigAssetReference {blueprintRigClass=&/M_MED_CadetFrog_Slide_Pack_CtrlRig.M_MED_CadetFrog_Slide_Pack_CtrlRig_C}
   }
   n2 = "Transform (Modify) Bone - Bone: dyn_needle_jnt"(
    ComponentPose=Local To Component(LocalPose=@n1)
    Alpha=Get Spinning Alpha() -> (Spinning Alpha:double)
    Translation="0,0,0"
    Rotation="0,0,-40"
    Scale="1,1,1"
   ) {
    rotationMode=BMM_Additive
    rotationSpace=BCS_BoneSpace
    translationMode=BMM_Additive
    alphaBoolBlend {blendOption=Linear}
    alphaScaleBiasClamp {bInterpResult=true, interpSpeedDecreasing=5, interpSpeedIncreasing=5}
    boneToModify {boneName=dyn_needle_jnt}
    rotation {roll=-40}
   }
   n3 = Anim Dynamics - Bone: dyn_tube!(SimulationSpace:AnimPhysSimSpaceType Enum=World, ComponentPose=@n2, Alpha=1, GravityOverride="0,0,-980", ExternalForce="0,0,0") -> (Pose) {
    angularDampingOverride=0.9
    angularSpringConstant=4000
    bAngularSpring=true
    bOverrideAngularDamping=true
    bUseGravityOverride=true
    gravityScale=0.35
    simulationSpace=World
    alphaBoolBlend {blendOption=Linear}
    boundBone {boneName=dyn_tube}
    componentAppliedLinearAccClamp=<9500 9500 9500>
    componentLinearAccScale=<1 1 1>
    componentLinearVelScale=<1 1 1>
    gravityOverride {z=-980}
    physicsBodyDefinitions [{
      boundBone {boneName=dyn_tube}
      boxExtents=<20 20 20>
      constraintSetup {angularLimitsMax=<3 3 3>, angularLimitsMin=<-3 -2 -5>, angularTarget=<1 0 0>}
      localJointOffset=<2 0 0>
     }]
    simSpaceSettings {maxAngularAcceleration=35, maxAngularVelocity=0.7, simSpaceAngularAlpha=1}
   }
   Output Pose(Result=Component To Local(ComponentPose=@n3)) {layerGroup=None}
  ]}
 EventGraph {nodes [
   Event BlueprintUpdateAnimation()
   > n1 = Sequence!(execute) -> (then_0, then_1, then_2:Exec, then_3:Exec)
   > n2 = "Gravity Override State Function (Nativized)"!(
    execute
    CharacterPartAnimInstance:Custom Character Part Anim Instance Object Reference=Self-Reference()
    GravityOverrideParameterSet:Fort Gravity Override Parameter Set Structure=Get Gravity Override Params_0() -> (Gravity Override Params_0:Fort Gravity Override Parameter Set Structure)
    RigidBodyAnimNodeParameterSet:Fort Rigid Body Anim Node Parameter Set Structure=Get RBAN Params_0() -> (RBAN Params_0:Fort Rigid Body Anim Node Parameter Set Structure)
    DebugDraw:bool=false
    GameTimeInSeconds:float=GetGameTimeInSeconds()
    DebugJointName:Name=root
    StateLogicLayerName:Name=none
   ) -> (
    then
    OutGravityOverride:Vector
    OutCurrentState:Name=None
    OutGravityOverrideParameters:Fort Gravity Override Parameters Structure
    OutRigidBodyAnimNodeParameters:Fort Rigid Body Anim Node Parameters Structure
   )
   > Set Gravity Override(Gravity Override=@n2)
   > Set RBAN Settings Out_0(RBAN Settings Out_0:Fort Rigid Body Anim Node Parameters Structure=@n2.OutRigidBodyAnimNodeParameters) -> (Output_Get:Fort Rigid Body Anim Node Parameters Structure)
   Branch(execute=@n1.then_1, Condition=Get bIsSprinting())
   > Delay(Duration=RandomFloatInRange(Min=0.5, Max=0))
   > Set Spinning Alpha(Spinning Alpha:double=RandomFloatInRange(Min=0, Max=1)) -> (Output_Get:double)
   n3 = Sequence!(execute=@n1.then_2) -> (then_0, then_1, then_2)
   n4 = Get bIsPlayingEmote()
   Set Sprinting(execute=@n3.then_1, Sprinting:bool=AND Boolean(A=Get bIsMoving2D(), B=NOT Boolean(A=@n4))) -> (Output_Get:bool)
   n5 = Get bIsMoving2D()
   n6 = Get bIsPlayingEmote()
   Set isIDLE(execute=@n3.then_0, isIDLE:bool=NOT Boolean(A=OR Boolean!(A=@n5, B=@n6, C:bool) -> (ReturnValue))) -> (Output_Get:bool)
   n7 = Branch(execute=@n1.then_3, Condition=Get bIsInFrontEnd())
   > Set AnimDynamic_SimulationSpace(AnimDynamic_SimulationSpace:Byte=1) -> (Output_Get:Byte=0)
   Set AnimDynamic_SimulationSpace(execute=@n7.else, AnimDynamic_SimulationSpace=2) -> (Output_Get=0)
   TryGetPawnOwner()
  ]}
]
variables [
 AnimDynamic_SimulationSpace {type=integer}
 Gravity Override {type=Vector}
 Gravity Override Params_0 {
  type=FortGravityOverrideParameterSet
  default=(Skydive_GoingUp=(GravityOverride=()),Skydive_FloatingDown=(GravityOverride=()),Skydive_DivingDown=(GravityOverride=()),Skydive_Parachuting=(GravityOverride=()),OnGround_Standing=(GravityOverride=()),OnGround_Moving=(GravityOverride=()),OnGround_Moving_HighRotationRate=(GravityOverride=()),OnGround_Crouch_Moving=(GravityOverride=()),EmoteOrMelee=(GravityOverride=()),Swimming=(GravityOverride=()),Front_End=(GravityOverride=()),DBNO=(GravityOverride=()),DBNO_Carried=(GravityOverride=()),OnGround_Falling=(GravityOverride=()),Ziplining=(GravityOverride=()),Skydive_Surfing=(GravityOverride=()),Riding_Standing=(GravityOverride=()),Riding_Moving=(GravityOverride=()),Riding_Moving_Targeting=(GravityOverride=()),Tactical_Sprint=(GravityOverride=()),Rail_Grinding=(GravityOverride=()),)
 }
 isIDLE {type=boolean}
 OnZipline {type=boolean}
 RBAN Params_0 {
  type=FortRigidBodyAnimNodeParameterSet
  default=(Skydive_GoingUp=(RigidBodyAnimNode=()),Skydive_FloatingDown=(RigidBodyAnimNode=()),Skydive_DivingDown=(RigidBodyAnimNode=()),Skydive_Parachuting=(RigidBodyAnimNode=()),OnGround_Standing=(RigidBodyAnimNode=()),OnGround_Moving=(RigidBodyAnimNode=()),OnGround_Moving_HighRotationRate=(RigidBodyAnimNode=()),OnGround_Crouch_Moving=(RigidBodyAnimNode=()),EmoteOrMelee=(RigidBodyAnimNode=()),Swimming=(RigidBodyAnimNode=()),Front_End=(RigidBodyAnimNode=()),DBNO=(RigidBodyAnimNode=()),DBNO_Carried=(RigidBodyAnimNode=()),OnGround_Falling=(RigidBodyAnimNode=()),Ziplining=(RigidBodyAnimNode=()),Skydive_Surfing=(RigidBodyAnimNode=()),Riding_Standing=(RigidBodyAnimNode=()),Riding_Moving=(RigidBodyAnimNode=()),Riding_Moving_Targeting=(RigidBodyAnimNode=()),Tactical_Sprint=(RigidBodyAnimNode=()),Rail_Grinding=(RigidBodyAnimNode=()),)
 }
 RBAN Settings Out_0 {type=FortRigidBodyAnimNodeParameters}
 Spinning Alpha {type=number}
 Sprinting {type=boolean}
]
